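# Niagara swing trail for the AgedChasm pickaxe, built from ribbon emitters that share system-level colors.
# MainCable, Tip and seven Accents trails use FN AnimTrail modules on socket indices, fading by curve, with time-of-day tint on Tip and Accents.
# ChainSourceRibbon bursts 16 linked particles that sample sockets and smooth-lerp a lagged position, whose offset from particle position drives ribbon facing and width, with a minimum width.
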NiagaraSystem /BRCosmetics/Weapons/FORT_Melee/Pickaxe_AgedChasm/FX/NS_Pickaxe_AgedChasm_SwingTrail
dependencies {
 usedDataInterfaces [
  &NiagaraDataInterfaceCurve
  &NiagaraDataInterfaceSkeletalMesh
  &NiagaraDataInterfaceSocketReader
  /Script/Niagara.NiagaraDataInterfaceArrayUInt8
  &NiagaraDataInterfaceArrayInt32
 ]
 usedDynamicInputs [&MakeCustomIntFromBool, &Add_Integer]
 usedModules [
  &self:SystemSpawnScript.SystemScriptSource.SystemScriptGraph.NiagaraNodeAssignment_2.SetVariables_#1
  &self:SystemSpawnScript.SystemScriptSource.SystemScriptGraph.NiagaraNodeAssignment_0.SetVariables_#2
  &self:SystemSpawnScript.SystemScriptSource.SystemScriptGraph.NiagaraNodeAssignment_1.SetVariables_#3
  &self:SystemSpawnScript.SystemScriptSource.SystemScriptGraph.NiagaraNodeAssignment_5.SetVariables_#4
  &self:SystemSpawnScript.SystemScriptSource.SystemScriptGraph.NiagaraNodeAssignment_3.SetVariables_#5
  &self:SystemSpawnScript.SystemScriptSource.SystemScriptGraph.NiagaraNodeAssignment_7.SetVariables_#6
  &self:SystemSpawnScript.SystemScriptSource.SystemScriptGraph.NiagaraNodeAssignment_8.SetVariables_#7
  &self:SystemSpawnScript.SystemScriptSource.SystemScriptGraph.NiagaraNodeAssignment_9.SetVariables_#8
  &self:SystemSpawnScript.SystemScriptSource.SystemScriptGraph.NiagaraNodeAssignment_10.SetVariables_#9
  &self:SystemSpawnScript.SystemScriptSource.SystemScriptGraph.NiagaraNodeAssignment_12.SetVariables_#10
  &SystemState
  &self:SystemSpawnScript.SystemScriptSource.SystemScriptGraph.NiagaraNodeAssignment_11.SetVariables_#11
  &self:SystemSpawnScript.SystemScriptSource.SystemScriptGraph.NiagaraNodeAssignment_13.SetVariables_#12
  &self:SystemSpawnScript.SystemScriptSource.SystemScriptGraph.NiagaraNodeAssignment_14.SetVariables_#13
  &self:SystemSpawnScript.SystemScriptSource.SystemScriptGraph.NiagaraNodeAssignment_4.SetVariables_#14
  &self:SystemSpawnScript.SystemScriptSource.SystemScriptGraph.NiagaraNodeAssignment_6.SetVariables_#15
 ]
}
emitters {
 Accents {
  emitterSpawnScript [&FN_RibbonIDOffsetUtility()]
  emitterUpdateScript [
   &EmitterState(Life Cycle Mode=::NewEnumerator1, Loop Behavior=::NewEnumerator1, Loop Duration=0.3, UseLoopDelay=false)
   &FN_SpawnAnimTrail(Number of Anim Trails=7)
  ]
  particleSpawnScript [
   &FN_InitializeAnimTrailBoneIndecies(
    First Bone or Socket Index:NiagaraInt32=&SelectIntFromArray(Direct Array Index=0, Int Selection Array=$System.AccentSocketIndices)
    Second Bone or Socket Index:NiagaraInt32=&SelectIntFromArray(Direct Array Index=0, Int Selection Array=$System.AccentSocketIndices)
   )
   &FN_InitializeAnimTrailBoneIndecies(
    AnimTrail Selection=1
    First Bone or Socket Index:NiagaraInt32=&SelectIntFromArray(Direct Array Index=1, Int Selection Array=$System.AccentSocketIndices)
    Second Bone or Socket Index:NiagaraInt32=&SelectIntFromArray(Direct Array Index=1, Int Selection Array=$System.AccentSocketIndices)
   ) {module=FN_InitializeAnimTrailBoneIndecies006}
   &FN_InitializeAnimTrailBoneIndecies(
    AnimTrail Selection=2
    First Bone or Socket Index:NiagaraInt32=&SelectIntFromArray(Direct Array Index=2, Int Selection Array=$System.AccentSocketIndices)
    Second Bone or Socket Index:NiagaraInt32=&SelectIntFromArray(Direct Array Index=2, Int Selection Array=$System.AccentSocketIndices)
   ) {module=FN_InitializeAnimTrailBoneIndecies005}
   &FN_InitializeAnimTrailBoneIndecies(
    AnimTrail Selection=3
    First Bone or Socket Index:NiagaraInt32=&SelectIntFromArray(Direct Array Index=3, Int Selection Array=$System.AccentSocketIndices)
    Second Bone or Socket Index:NiagaraInt32=&SelectIntFromArray(Direct Array Index=3, Int Selection Array=$System.AccentSocketIndices)
   ) {module=FN_InitializeAnimTrailBoneIndecies004}
   &FN_InitializeAnimTrailBoneIndecies(
    AnimTrail Selection=4
    First Bone or Socket Index:NiagaraInt32=&SelectIntFromArray(Direct Array Index=4, Int Selection Array=$System.AccentSocketIndices)
    Second Bone or Socket Index:NiagaraInt32=&SelectIntFromArray(Direct Array Index=4, Int Selection Array=$System.AccentSocketIndices)
   ) {module=FN_InitializeAnimTrailBoneIndecies003}
   &FN_InitializeAnimTrailBoneIndecies(
    AnimTrail Selection=5
    First Bone or Socket Index:NiagaraInt32=&SelectIntFromArray(Direct Array Index=5, Int Selection Array=$System.AccentSocketIndices)
    Second Bone or Socket Index:NiagaraInt32=&SelectIntFromArray(Direct Array Index=5, Int Selection Array=$System.AccentSocketIndices)
   ) {module=FN_InitializeAnimTrailBoneIndecies002}
   &FN_InitializeAnimTrailBoneIndecies(
    AnimTrail Selection=6
    First Bone or Socket Index:NiagaraInt32=&SelectIntFromArray(Direct Array Index=6, Int Selection Array=$System.AccentSocketIndices)
    Second Bone or Socket Index:NiagaraInt32=&SelectIntFromArray(Direct Array Index=6, Int Selection Array=$System.AccentSocketIndices)
   ) {module=FN_InitializeAnimTrailBoneIndecies001}
   &FN_InitializeAnimTrail(
    Color=&MultiplyLinearColor(
     Linear Color=$System.AccentColor
     Scale Factor=&MakeVector4_FromLinearColor(LinearColor=&FNTimeOfDayColor(
       Night Color=<rgba 0.1218 0.1931 0.35 1>
       Morning Color=<rgba 0.5 0.4 0.4074 1>
       Day Color=<rgba 0.7 0.7 0.7 1>
       Evening Color=<rgba 0.5 0.3044 0.3315 1>
      ))
    )
    Lifetime=0.15
   )
   /SetVariables(Particles.RibbonWidth=8) {isSetParameters=true, module=SetVariables_#16}
  ]
  particleUpdateScript [
   &FN_UpdateAnimTrail(Locked Particle RibbonWidth:NiagaraFloat=$Particles.RibbonWidth)
   &ScaleColor(
    Scale Alpha=&Multiply_Float(
     A=&FloatFromCurve(
      FloatCurve=DataInterface() {propertyValues {Curve {keys [<at 0 1 RCIM_Cubic>, {arriveTangent=-1, interpMode=RCIM_Cubic, leaveTangent=-1, time=0.4, value=0.5}, <at 1 0 RCIM_Cubic>]}}}
      Scale Curve=$System.SwingGeneralFade
     )
     B=0.1
    )
    Scale RGB=<0.8 0.8 0.8>
    ScaleA=true
    ScaleRGB=true
   )
   /SetVariables(Particles.RibbonWidth=&Multiply_Float(
     A=&FloatFromCurve(FloatCurve=DataInterface() {propertyValues {Curve {keys [<at 0.05 1 RCIM_Cubic>,<at 1 0 RCIM_Cubic>]}}})
     B=$Particles.Initial.RibbonWidth
    )) {isSetParameters=true}
  ]
  rendererClasses [&NiagaraRibbonRendererProperties]
  renderers [
   {rendererClass=&NiagaraRibbonRendererProperties}
   {rendererClass=&NiagaraRibbonRendererProperties, rendererIndex=1}
  ]
 }
 ChainSourceRibbon {
  emitterUpdateScript [
   &EmitterState(Life Cycle Mode=::NewEnumerator1, Loop Behavior=::NewEnumerator1, Loop Duration=0.75, UseLoopDelay=false)
   &SpawnBurst_Instantaneous(Spawn Count=16, Use Spawn Probability=false, UseLoopCountLimit=false)
  ]
  particleSpawnScript [
   &InitializeParticle(
    Color=$System.CableColor
    Lifetime=0.4
    Sprite Size Mode=::NewEnumerator3
    Uniform Sprite Size=4
    UsePositionOffset=false
    Write Material Random=true
   )
   /SetVariables(Particles.RibbonWidth=47.2012) {isSetParameters=true}
   /SetVariables(Particles.RibbonLinkOrder:NiagaraFloat=&ReturnNormalizedExecIndex()) {isSetParameters=true}
   &SocketLocation(
    SamplePosOffset=false
    Socket Index=&MakeIntFromFloat(Float=&Multiply_Float(A=&ReturnNormalizedExecIndex(), B=16))
    Socket Reader=$System.Socket Reader
    Socket Sampling Mode=::NewEnumerator3
    VelocitySpeedLimit=false
   )
   /SetVariables(Particles.LaggedPosition:NiagaraPosition=$Output.SocketLocation.SampledPosition) {isSetParameters=true}
  ]
  particleUpdateScript [
   &ParticleState()
   &SocketLocation(
    SamplePosOffset=false
    Socket Index=&MakeIntFromFloat(Float=&Multiply_Float(A=&ReturnNormalizedExecIndex(), B=16))
    Socket Reader=$System.Socket Reader
    Socket Sampling Mode=::NewEnumerator3
    VelocitySpeedLimit=false
   ) {module=SocketLocation001}
   /SetVariables(Particles.LaggedPosition:NiagaraPosition=/Niagara/DynamicInputs/Lerp/SmoothLerpOverTime_Position(
     Convergence Rate:NiagaraFloat=2
     Smooth Value:NiagaraPosition=$Particles.LaggedPosition
     Target Value:NiagaraPosition=$Output.SocketLocation001.SampledPosition
    )) {isSetParameters=true}
   /SetVariables(Particles.RibbonWidth=&Multiply_Float(
     A=&Max_Float(
      A:NiagaraFloat=&Multiply_Float(A=&DistanceBetweenPositions(A=$Particles.LaggedPosition, B=$Particles.Position), B=0.5)
      B:NiagaraFloat=60.3518
     )
     B=1
    )) {isSetParameters=true}
   /SetVariables(Particles.RibbonFacing=&NormalizeVector(
     Vector To Normalize=&SubtractVectorFromPosition(Position=$Particles.Position, Vector=Linked() {linkedVariable {name=Particles.LaggedPosition, type=NiagaraPosition}})
     Safe Normalize=true
     Safe Fallback Threshold=0.2
     Safe Fallback Vector=<1 0 0>
    )) {isSetParameters=true}
   &ScaleColor(
    Scale Alpha=&FloatFromCurve(
     FloatCurve=DataInterface() {propertyValues {Curve {keys [<at 0 1 RCIM_Cubic>, {arriveTangent=-1, interpMode=RCIM_Cubic, leaveTangent=-1, time=0.6, value=0.9}, <at 1 0 RCIM_Cubic>]}}}
     Scale Curve=0.4
    )
    ScaleA=true
    ScaleRGB=true
   )
  ]
  rendererClasses [&NiagaraRibbonRendererProperties]
  renderers [{rendererClass=&NiagaraRibbonRendererProperties}]
 }
 MainCable {
  emitterSpawnScript [&FN_RibbonIDOffsetUtility()]
  emitterUpdateScript [
   &EmitterState()
   &FN_SpawnAnimTrail()
  ]
  particleSpawnScript [
   &FN_InitializeAnimTrailBoneIndecies(First Bone or Socket Index=14)
   /SetVariables(Particles.RibbonWidth=100) {isSetParameters=true, module=SetVariables_#16}
   &FN_InitializeAnimTrail(Color=&MultiplyLinearColor(Linear Color=$System.CableColor, Scale Factor=<1 1 1 1>), Lifetime=0.15)
  ]
  particleUpdateScript [
   &FN_UpdateAnimTrail(Locked Particle RibbonWidth:NiagaraFloat=$Particles.RibbonWidth)
   &ScaleColor(
    Scale Alpha=&Multiply_Float(
     A=&FloatFromCurve(
      FloatCurve=DataInterface() {propertyValues {Curve {keys [<at 0 1 RCIM_Cubic>, {arriveTangent=-1, interpMode=RCIM_Cubic, leaveTangent=-1, time=0.4, value=0.375}, <at 1 0 RCIM_Cubic>]}}}
      Scale Curve=$System.SwingGeneralFade
     )
     B=0.35
    )
    ScaleA=true
    ScaleRGB=true
   )
   /SetVariables(Particles.TipWidth:NiagaraFloat=&FloatFromCurve(FloatCurve=DataInterface() {propertyValues {Curve {keys [<at 0 1>,<at 1 0>]}}}, Scale Curve=10)) {isSetParameters=true, module=SetVariables_#17}
   &ScaleColor(Scale RGB=&VectorFromFloat(Value=$System.TODFloat), ScaleA=false, ScaleRGB=true) {enabled=false, module=ScaleColor001}
   /SetVariables(Particles.SoftColor:LinearColor=&MakeLinearColorFromVectorAndFloat(
     "Vector (RGB)"=&Multiply_VectorByFloat(
      Vector=&Multiply_Vector(A=<1 0.7492 0.531>, B=&MakeVectorFromLinearColorRGB(Color=$System.TODColor))
      Float=$System.TODFloat
     )
     "Float (Alpha)"=&Multiply_Float(
      A=&FloatFromCurve(
       FloatCurve=DataInterface() {propertyValues {Curve {keys [{interpMode=RCIM_Cubic, leaveTangent=-1.7551, tangentMode=RCTM_Break, time=0, value=1}, <at 0.6 0>]}}}
       Scale Curve=$System.SwingGeneralFade
      )
      B=0.8
     )
    )) {enabled=false, isSetParameters=true}
   /SetVariables() {enabled=false, isSetParameters=true}
  ]
  rendererClasses [&NiagaraRibbonRendererProperties]
  renderers [
   {rendererClass=&NiagaraRibbonRendererProperties}
   {rendererClass=&NiagaraRibbonRendererProperties, rendererIndex=1}
  ]
 }
 Tip {
  emitterSpawnScript [&FN_RibbonIDOffsetUtility()]
  emitterUpdateScript [
   &EmitterState()
   &FN_SpawnAnimTrail()
  ]
  particleSpawnScript [
   &FN_InitializeAnimTrailBoneIndecies(First Bone or Socket Index=1)
   &FN_InitializeAnimTrail(
    Color=&MultiplyLinearColor(
     Linear Color=$System.AccentColor
     Scale Factor=&MakeVector4_FromLinearColor(LinearColor=&FNTimeOfDayColor(Night Color=<rgba 0.35 0.35 0.35 1>, Morning Color=<rgba 0.75 0.75 0.75 1>, Day Color=<rgba 1 1 1 1>, Evening Color=<rgba 0.7 0.7 0.7 1>))
    )
    Lifetime=0.2
   )
  ]
  particleUpdateScript [
   &FN_UpdateAnimTrail()
   &ScaleColor(
    Scale Alpha=&Multiply_Float(
     A=&FloatFromCurve(
      FloatCurve=DataInterface() {propertyValues {Curve {keys [<at 0 1 RCIM_Cubic>, {arriveTangent=-1, interpMode=RCIM_Cubic, leaveTangent=-1, time=0.4, value=0.2}, <at 1 0 RCIM_Cubic>]}}}
      Scale Curve=$System.SwingGeneralFade
     )
     B=1
    )
    ScaleA=true
    ScaleRGB=true
   )
   /SetVariables(Particles.TipWidth:NiagaraFloat=&FloatFromCurve(FloatCurve=DataInterface() {propertyValues {Curve {keys [<at 0 1>,<at 1 0>]}}}, Scale Curve=5)) {isSetParameters=true, module=SetVariables_#17}
   /SetVariables(Particles.CamFacingWidth:NiagaraFloat=&Multiply_Float(A=$Particles.RibbonWidth, B=0.5)) {isSetParameters=true}
   /SetVariables(Particles.RibbonWidth=&FloatFromCurve(FloatCurve=DataInterface() {propertyValues {Curve {keys [<at 0 0>,<at 0.1 1>,<at 1 0>]}}}, Scale Curve=5)) {isSetParameters=true}
  ]
  rendererClasses [&NiagaraRibbonRendererProperties]
  renderers [
   {rendererClass=&NiagaraRibbonRendererProperties}
   {rendererClass=&NiagaraRibbonRendererProperties, rendererIndex=1}
   {rendererClass=&NiagaraRibbonRendererProperties, rendererIndex=2}
  ]
 }
}
summary {
 systemName=NS_Pickaxe_AgedChasm_SwingTrail
 emitters [
  {emitterName=MainCable, rendererClasses [&NiagaraRibbonRendererProperties]}
  {emitterName=Tip, rendererClasses [&NiagaraRibbonRendererProperties]}
  {emitterName=ChainSourceRibbon, rendererClasses [&NiagaraRibbonRendererProperties]}
  {emitterName=Accents, rendererClasses [&NiagaraRibbonRendererProperties]}
 ]
 userVariables [
  User.IsOffhand {type=NiagaraBool}
  User.RibbonIDOffset {type=NiagaraInt32}
  User.SwingSectionIndex {type=NiagaraInt32}
  User.TrailColor {type=LinearColor, description="", defaultValue {struct=LinearColor, value=<rgba 1 0.439 0.1161 0.35>}}
  User.UseBaseEnd {type=NiagaraBool}
  User.VariantIndex {type=NiagaraFloat, defaultValue=0, description=""}
 ]
}
